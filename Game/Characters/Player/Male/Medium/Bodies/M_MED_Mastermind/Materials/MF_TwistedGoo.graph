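# Material function for the Mastermind "Twisted" skin effect, blending animated dark goo over input material attributes by Mask, toggled by UseTwistedGoo.
# Panning triplanar masks on pre-skinned position drive colour variation and distort a projected normal, which feeds a view-facing emissive highlight.
# IsTwistedWrap desaturates the base and switches vehicle/mech tiling scaling.
# Goo normals angle-correct blend onto the base; the blend's UseA vertex setting discards the computed goo vertex offset.
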MaterialFunction /Game/Characters/Player/Male/Medium/Bodies/M_MED_Mastermind/Materials/MF_TwistedGoo
graph {nodes [
  n1 = ScalarParameter() {defaultValue=7.1525, parameterName=Fresnel}
  n2 = TextureObject() {samplerType=SAMPLERTYPE_Normal, texture=/Game/Characters/Player/Male/Medium/Bodies/M_MED_Mastermind/Textures/T_M_MED_Mastermind_Twisted_N}
  n3 = LinearInterpolate(
   A=LinearInterpolate(
    A=ScalarParameter() {defaultValue=1, parameterName=Default_Tiling_Override}
    B=ScalarParameter() {defaultValue=1, parameterName=Vehicle_TilingOveride}
    Alpha=FunctionInput() {bUsePreviewValueAsDefault=true, inputName=isVehicle, inputType=FunctionInput_Scalar, sortPriority=32}
   )
   B=ScalarParameter() {defaultValue=1, parameterName=Mech_Tiling_Override}
   Alpha=FunctionInput() {bUsePreviewValueAsDefault=true, inputName=isMech, inputType=FunctionInput_Scalar, sortPriority=33}
  )
  n4 = AppendVector(
   A=AppendVector(A=ScalarParameter() {defaultValue=84.6447, parameterName=Size_Normal}, B=ScalarParameter() {defaultValue=82.35, parameterName=Size_Normal2})
   B=ScalarParameter() {defaultValue=33.2833, parameterName=Size_Normal3}
  )
  n5 = StaticSwitchParameter(True=Multiply(A=@n3, B=@n4), False=@n4) {parameterName=IsTwistedWrap}
  n6 = Constant3Vector() {constant=<rgba 0 0 1 0>}
  n7 = PreSkinnedNormal()
  n8 = VertexInterpolator(VS=@n7)
  n9 = Normalize(VectorInput=Transform(None=@n8.PS) {transformSourceType=TRANSFORMSOURCE_Local, transformType=TRANSFORM_Tangent})
  n10 = PreSkinnedPosition()
  n11 = VertexInterpolator(VS=@n10)
  n12 = ScalarParameter() {defaultValue=6, parameterName=Distortion}
  n13 = TextureObject() {samplerType=SAMPLERTYPE_Masks, texture=/Game/Characters/Player/Male/Medium/Bodies/M_MED_Mastermind/Textures/T_M_MED_Mastermind_Twisted_M}
  n14 = ScalarParameter() {defaultValue=33.9662, parameterName=Size_DistortionMask}
  n15 = ComponentMask(None=VectorParameter() {parameterName=Pan_Mask_YZ, defaultValue=<rgba -0.0857 0.1333 0 1>}.RGB) {g=true, r=true}
  n16 = ScalarParameter() {defaultValue=0.4286, parameterName=PanSpeedMask}
  n17 = ComponentMask(None=VectorParameter() {parameterName=Pan_Mask_XZ, defaultValue=<rgba 0.0762 0.1333 0 1>}.RGB) {g=true, r=true}
  n18 = ComponentMask(None=VectorParameter() {parameterName=Pan_Mask_XY, defaultValue=<rgba 0.0286 -0.05 0 1>}.RGB) {g=true, r=true}
  n19 = MaterialFunctionCall!(
   Export Float 4
   TextureObject=@n13
   TextureSize=StaticSwitchParameter(True=@n14, False=Multiply(A=@n14, B=@n3)) {parameterName=IsTwistedWrap}
   WorldPosition=@n11.PS
   World Space Normal=@n8.PS
   ProjectionTransitionContrast
   Pan_YZ=Multiply(A=@n15, B=@n16)
   Pan_XZ=Multiply(A=@n17, B=@n16)
   Pan_XY=Multiply(A=@n18, B=@n16)
  ) {materialFunction=/Game/Weapons/WeaponSkins/Wraps/MaterialFunctions/WorldAlignedTexture_ComplexPan}
  n20 = Multiply(A=AppendVector(A=AppendVector(A=@n12, B=@n12), B=@n12), B=ComponentMask(None=@n19.XYZ Texture) {g=true})
  n21 = ScalarParameter() {defaultValue=0.5524, parameterName=PanSpeedNormal}
  n22 = Multiply(A=@n15, B=@n21)
  n23 = Multiply(A=@n17, B=@n21)
  n24 = Multiply(A=@n18, B=@n21)
  n25 = MaterialFunctionCall(
   Normal=MaterialFunctionCall!(
    TextureObject=@n2
    TextureSize=@n5
    Normal=StaticSwitchParameter(True=@n6, False=@n9) {parameterName=UseTwistedCorrectNormals}
    WorldSpace
    ProjectionTransitionContrast=ScalarParameter() {parameterName=ProjectionContrast}
    WorldPosition=Add(A=@n11.PS, B=@n20)
    Use High Quality Normals=StaticBoolParameter() {defaultValue=true}
    Pan_YZ=@n22
    Pan_XZ=@n23
    Pan_XY=@n24
   ) {materialFunction=/Game/Weapons/WeaponSkins/Wraps/MaterialFunctions/WorldAlignedNormal_ComplexPan}.XYZ Texture
   Flatness=ScalarParameter() {defaultValue=0.2476, parameterName=NormalFlatten}
  ) {materialFunction=&FlattenNormal}
  n26 = Transform(None=@n25.Result)
  n27 = DotProduct(A=CameraVectorWS(), B=Normalize(VectorInput=@n26))
  n28 = Multiply(
   A=ScalarParameter() {defaultValue=0.2762, parameterName=FresnelBirghtness2}
   B=Power(Base=@n27, Exp=ScalarParameter() {defaultValue=442.9701, parameterName=Fresnel2})
  )
  Add(
   A=Multiply(A=ScalarParameter() {defaultValue=0.0667, parameterName=FresnelBirghtness}, B=Fresnel(ExponentIn=@n1, Normal=@n26) {baseReflectFraction=0})
   B=@n28
  )
  n29 = FunctionInput(Preview=SetMaterialAttributes!(MaterialAttributes)) {bUsePreviewValueAsDefault=true, inputType=FunctionInput_MaterialAttributes}
  n30 = GetMaterialAttributes(None=@n29) {attributeGetTypes [#1]}
  n31 = Desaturation(None=@n30.BaseColor, Fraction=ScalarParameter() {defaultValue=1, parameterName=Desat Diffuse})
  n32 = ScalarParameter() {defaultValue=-0.4238, parameterName=VertexOffsetDistance}
  n33 = ComponentMask(None=MaterialFunctionCall(TextureObject=@n13, TextureSize=@n5, WorldPosition=@n10, World Space Normal=@n7, Pan_YZ=@n22, Pan_XZ=@n23, Pan_XY=@n24) {materialFunction=&WorldAlignedTexture_ComplexPan}.XYZ Texture) {b=true}
  n34 = LinearInterpolate(A=@n32, B=ScalarParameter() {defaultValue=0.791, parameterName=VertexOffsetDistance_Max}, Alpha=@n33)
  n35 = VectorParameter() {parameterName=OffsetDirection, defaultValue=<rgba 2 -1 2 1>}
  n36 = SetMaterialAttributes!(
   MaterialAttributes
   BaseColor=LinearInterpolate(
    A=VectorParameter() {parameterName=Color, defaultValue=<rgba 0 0 0 1>}.RGB
    B=VectorParameter() {parameterName=Color2, defaultValue=<rgba 0.0022 0.0017 0.005 1>}.RGB
    Alpha=ComponentMask(None=@n19.XYZ Texture) {b=true}
   )
   Specular=ScalarParameter() {defaultValue=0.3238, parameterName=Specular}
   Roughness=ScalarParameter() {defaultValue=0.33, parameterName=Roughness}
   EmissiveColor=@n28
   Normal=@n25.Result
   WorldPositionOffset=Multiply(A=Clamp(None=@n34) {clampMode=CMODE_ClampMin}, B=Transform(None=@n35.RGB) {transformSourceType=TRANSFORMSOURCE_Local})
  ) {attributeSetTypes [
    #1
    #2
    #3
    #4
    #5
    #6
   ]}
  n37 = FunctionInput(Preview=Constant() {r=1}) {bUsePreviewValueAsDefault=true, inputName=Mask, inputType=FunctionInput_Scalar, sortPriority=3}
  n38 = BlendMaterialAttributes(
   A=StaticSwitchParameter(
    True=SetMaterialAttributes!(MaterialAttributes=@n30.MaterialAttributes, BaseColor=@n31) {attributeSetTypes [#1]}
    False=@n29
   ) {parameterName=IsTwistedWrap}
   B=@n36
   Alpha=@n37
  ) {vertexAttributeBlendType=UseA}
  n39 = GetMaterialAttributes(None=@n29) {attributeGetTypes [#5]}
  n40 = LinearInterpolate(
   A=@n39.Normal
   B=MaterialFunctionCall(BaseNormal=@n39.Normal, AdditionalNormal=GetMaterialAttributes(None=@n38) {attributeGetTypes [#5]}.Normal) {materialFunction=&BlendAngleCorrectedNormals}.Result
   Alpha=@n37
  )
  FunctionOutput(None=StaticSwitchParameter(True=SetMaterialAttributes!(MaterialAttributes=@n38, Normal=@n40) {attributeSetTypes [#5]}, False=@n29) {group=Skin FX, parameterName=UseTwistedGoo})
  Multiply(B=@n27) {constA=-1}
  Multiply(A=@n32, B=Subtract(A=@n33, B=ScalarParameter() {defaultValue=0.3524, parameterName=BiasVertexoffset}))
  Normalize(VectorInput=@n35.RGB)
  VertexNormalWS()
 ]}
